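# Physics asset for the Cat Burglar Summer Male pickaxe mesh, covering its dangling rope and two disk bones.
# A small sphere on the rope base starts a chain of capsule bodies running through the rope parent and three rope segments.
# Both disk capsules branch in parallel from the last rope segment.
# Every joint shares the same narrow swing cone and very tight twist limit.
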PhysicsAsset /BRCosmetics/Weapons/FORT_Melee/Pickaxe_Cat_Burglar_Summer_Male/Meshes/Cat_Burglar_Summer_Male_Axe_Physics
bodies {
 dyn_disk_01 [{shapeType=Capsule, length=3.3439, radius=1.4135, center=<3.0003 0 0>, rotation=<rot 90 0 0>}]
 dyn_disk_02 [{shapeType=Capsule, length=3.9689, radius=1.6531, center=<3.2791 0.0001 0>, rotation=<rot 90 0 0>}]
 dyn_rope_01 [{shapeType=Capsule, length=3.1858, radius=1.0358, center=<2.5722 0 0>, rotation=<rot 90 0 0>}]
 dyn_rope_02 [{shapeType=Capsule, length=3.1327, radius=0.9878, center=<2.5032 0 0>, rotation=<rot 90 0 0>}]
 dyn_rope_03 [{shapeType=Capsule, length=2.7996, radius=0.9993, center=<2.2121 0 0>, rotation=<rot 90 0 0>}]
 dyn_rope_base [{shapeType=Sphere, radius=0.5}]
 dyn_rope_parent [{shapeType=Capsule, length=3.6895, radius=0.9684, center=<2.3615 0.0001 0>, rotation=<rot 90 0 0>}]
}
constraints [
 {bone1Name=dyn_rope_parent, bone2Name=dyn_rope_base, swing1LimitDegrees=20, swing2LimitDegrees=20, twistLimitDegrees=3}
 {bone1Name=dyn_rope_01, bone2Name=dyn_rope_parent, swing1LimitDegrees=20, swing2LimitDegrees=20, twistLimitDegrees=3}
 {bone1Name=dyn_rope_02, bone2Name=dyn_rope_01, swing1LimitDegrees=20, swing2LimitDegrees=20, twistLimitDegrees=3}
 {bone1Name=dyn_rope_03, bone2Name=dyn_rope_02, swing1LimitDegrees=20, swing2LimitDegrees=20, twistLimitDegrees=3}
 {bone1Name=dyn_disk_01, bone2Name=dyn_rope_03, swing1LimitDegrees=20, swing2LimitDegrees=20, twistLimitDegrees=3}
 {bone1Name=dyn_disk_02, bone2Name=dyn_rope_03, swing1LimitDegrees=20, swing2LimitDegrees=20, twistLimitDegrees=3}
]
